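# Additive unlit Niagara ribbon and sprite material for the front trail effect of the YuzuCrankTide pickaxe.
# Particle color RGB feeds the shared MF_FX_EmissiveAdjustments function for emissive.
# Particle alpha is multiplied by an inverted UV gradient, with USE_VERTICAL_MASK blending from U to V.
# USE_OPACITY_MULT blends in an extra mask from the TX_Opacity_Mult red channel, which defaults to a SpireTend pickaxe texture.
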Material /BRCosmetics/Weapons/FORT_Melee/Pickaxe_YuzuCrankTide/FX/Mat/M_YuzuCrankTide_TrailFront_EmissiveAdjustment
graph {
 properties {
  bEnableResponsiveAA=true
  bOutputTranslucentVelocity=true
  bUsedWithNiagaraRibbons=true
  bUsedWithNiagaraSprites=true
  blendMode=BLEND_Additive
  pixelDepthOffsetMode=PDOM_Legacy
  shadingModel=MSM_Unlit
  usedShadingModels=Unlit
 }
 nodes [
  n1 = ParticleColor()
  MaterialFunctionCall(Emissive=@n1.RGB) {materialFunction=&MF_FX_EmissiveAdjustments}
  n2 = OneMinus(None=ComponentMask(None=TextureCoordinate()) {r=true})
  n3 = OneMinus(None=ComponentMask(None=TextureCoordinate()) {g=true})
  n4 = ScalarParameter() {parameterName=USE_VERTICAL_MASK}
  Multiply(
   A=Multiply(A=@n1.A, B=LinearInterpolate(A=@n2, B=@n3, Alpha=@n4))
   B=LinearInterpolate(
    B=TextureSampleParameter2D() {parameterName=TX_Opacity_Mult, texture=/BRCosmetics/Weapons/FORT_Melee/Pickaxe_SpireTend/FX/T_Pickaxe_SpireTend_Front}.R
    Alpha=ScalarParameter() {parameterName=USE_OPACITY_MULT}
   ) {constA=1}
  )
 ]
}
